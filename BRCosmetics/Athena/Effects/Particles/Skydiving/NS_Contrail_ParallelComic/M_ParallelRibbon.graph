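# Translucent unlit spider-web material for Niagara ribbons and mesh particles in the ParallelComic skydiving contrail.
# Dynamic-parameter tiling and offset drive web textures, with a scalar parameter blending in SplineThicken UV orientation, normals and vertex offset.
# Emissive fakes sun, rim and backlight scattering (backlight dropped on low quality), tinted by time of day or frontend color.
# Opacity sharpens a thresholded mask with height-blended end dissolve, particle alpha and depth fade.
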Material /BRCosmetics/Athena/Effects/Particles/Skydiving/NS_Contrail_ParallelComic/M_ParallelRibbon
graph {
 properties {
  bUsedWithNiagaraMeshParticles=true
  bUsedWithNiagaraRibbons=true
  blendMode=BLEND_Translucent
  pixelDepthOffsetMode=PDOM_Legacy
  shadingModel=MSM_Unlit
  twoSided=true
  usedShadingModels=Unlit
 }
 nodes [
  n1 = DynamicParameter() {parameterIndex=1, defaultValue=<rgba 1.016 1 1 1>, paramNames [Param1, Param2, Spline_WidthBase, Spline_WidthTip]}
  n2 = DynamicParameter() {defaultValue=<rgba 1.016 0 1 1>, paramNames [Tiling,Offset,Param3,Param4]}
  n3 = Constant()
  n4 = AppendVector(A=@n2.Offset, B=@n3)
  n5 = AppendVector(A=@n3, B=@n2.Offset)
  n6 = ScalarParameter() {parameterName=Use with SplineThicken}
  n7 = ScalarParameter() {defaultValue=1.0159, parameterName=Tiling}
  n8 = Multiply(
   A=Add(A=TextureCoordinate(), B=LinearInterpolate(A=@n4, B=@n5, Alpha=@n6))
   B=LinearInterpolate(A=AppendVector(A=@n2.Tiling, B=@n7), B=AppendVector(A=@n7, B=@n2.Tiling), Alpha=ScalarParameter() {parameterName=Use with SplineThicken})
  )
  n9 = LinearInterpolate(
   A=MaterialFunctionCall(XY=@n8) {materialFunction=&Swizzle}.YX
   B=@n8
   Alpha=ScalarParameter() {parameterName=Use with SplineThicken}
  )
  n10 = TextureSample(UVs=@n9) {samplerType=SAMPLERTYPE_Normal, texture=&/SpiderWebs_N}
  n11 = MaterialFunctionCall(
   WidthBase=@n1.Spline_WidthBase
   WidthTip=@n1.Spline_WidthBase
   WorldPosition=WorldPosition().XYZ
   AngleCorrectedNormals=StaticBool() {value=true}
   AdditionalNormal=@n10.RGB
  ) {materialFunction=&SplineThicken}
  LinearInterpolate(A=Constant(), B=@n11.WorldPositionOffset, Alpha=ScalarParameter() {parameterName=Use with SplineThicken})
  n12 = TextureSample(UVs=@n9) {samplerType=SAMPLERTYPE_Masks, texture=&/SpiderWebs_M}
  n13 = LinearInterpolate(
   A=VectorParameter() {parameterName=Color, defaultValue=<rgba 0.3613 0.4678 0.5647 1>}.RGB
   B=VectorParameter() {parameterName=Color2, defaultValue=<rgba 0.0561 0.0844 0.1301 1>}.RGB
   Alpha=@n12.R
  )
  n14 = CollectionParameter() {
   collection=&FortniteMaterialParameters
   parameterName=SunlightDirectionalVector
  }
  n15 = ComponentMask(None=@n14) {b=true, g=true, r=true}
  n16 = DotProduct(A=@n15, B=Multiply(A=CameraVectorWS()) {constB=-1})
  n17 = ScalarParameter() {defaultValue=36.3085, parameterName=BackLightPower}
  n18 = Multiply(A=Saturate(None=Power(Base=@n16, Exp=@n17)), B=ScalarParameter() {defaultValue=1.3555, parameterName=BackLightBrightness})
  n19 = OneMinus(None=Multiply(A=@n12.R, B=ScalarParameter() {defaultValue=0.9481, parameterName=FakeSSSopacity}))
  n20 = ScalarParameter() {parameterName=Use with SplineThicken}
  n21 = Normalize(VectorInput=MaterialFunctionCall(
    Normal=LinearInterpolate(A=@n10.RGB, B=@n11.Normal, Alpha=@n20)
    Flatness=ScalarParameter() {defaultValue=0.5312, parameterName=Normal_Flatten}
   ) {materialFunction=&FlattenNormal}.Result)
  n22 = OneMinus(None=DotProduct(A=@n21, B=CameraVectorWS()))
  n23 = ScalarParameter() {defaultValue=2.0916, parameterName=HiglightPower}
  n24 = DotProduct(A=@n21, B=@n15)
  n25 = ScalarParameter() {defaultValue=6.145, parameterName=SunLightPower}
  n26 = Add(
   A=Multiply(A=Power(Base=@n22, Exp=@n23), B=ScalarParameter() {defaultValue=0.0163, parameterName=HiglightBrightness})
   B=Multiply(A=Power(Base=@n24, Exp=@n25), B=VectorParameter() {parameterName=SunLightColor, defaultValue=<rgba 0.3594 0.3425 0.2639 1>}.RGB)
  )
  n27 = QualitySwitch(Default=Add(A=Multiply(A=@n18, B=@n19), B=@n26), Low=@n26)
  n28 = MaterialFunctionCall(
   Color_Morning=VectorParameter() {parameterName=TintColor_morning, defaultValue=<rgba 0.6979 0.2578 0.1388 1>}.RGB
   Color_Day=VectorParameter() {parameterName=TintColor, defaultValue=<rgba 0.718 0.8395 1 1>}.RGB
   Color_Evening=VectorParameter() {parameterName=TintColor_Evening, defaultValue=<rgba 1 0.4053 0.2142 1>}.RGB
   Color_Night=VectorParameter() {parameterName=TintColor_Night, defaultValue=<rgba 0.4088 0.5962 0.6146 1>}.RGB
  ) {materialFunction=&Color_Time_of_Day}
  n29 = VectorParameter() {parameterName=FrontendLightColor, defaultValue=<rgba 2.8335 3.3552 4 1>}
  n30 = ScalarParameter() {parameterName=isFrontend}
  MaterialFunctionCall(Emissive=Multiply(A=Add(A=@n13, B=@n27), B=LinearInterpolate(A=@n28.Result, B=@n29.RGB, Alpha=@n30))) {materialFunction=&MF_FX_EmissiveAdjustments}
  n31 = LinearInterpolate(
   A=ScalarParameter() {defaultValue=0.4789, parameterName=Opacity1}
   B=ScalarParameter() {defaultValue=0.7938, parameterName=Opacity2}
   Alpha=@n12.R
  )
  n32 = ComponentMask(None=TextureCoordinate() {coordinateIndex=1}) {g=true}
  n33 = OneMinus(None=Power(Base=OneMinus(None=@n32), Exp=ScalarParameter() {defaultValue=37.0469, parameterName=EndOpacity3}))
  n34 = Multiply(
   A=@n12.G
   B=MaterialFunctionCall(
    Transition Phase=Saturate(None=@n33)
    Height Texture=@n12.R
    Contrast=ScalarParameter() {defaultValue=0.1232, parameterName=EndOpacityContrast}
   ) {materialFunction=&HeightLerp}.Alpha
  )
  n35 = ScalarParameter() {defaultValue=0.5488, parameterName=OpacityThreshold}
  n36 = Saturate(None=Multiply(A=Subtract(A=@n34, B=@n35), B=ScalarParameter() {defaultValue=32.1132, parameterName=OpacityDistanceField}))
  Multiply(A=ParticleColor().A, B=DepthFade(Opacity=Multiply(A=@n31, B=@n36), FadeDistance=ScalarParameter() {defaultValue=5, parameterName=DepthFade}))
 ]
}
