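# Skydiving contrail Niagara system that trails web strands from two sampled player skeleton points, with "thwip" projectiles, impact sprites, hand-wrap meshes and smoke.
# Webs emitters build pendulum-constrained ribbons that extend and fade in on a system timeline.
# Cross_Webs emitters read those particles through attribute readers to hang sagging ribbons between strands, using upward gravity.
# A system Dive value increases web noise and shifts strand offsets.
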NiagaraSystem /BRCosmetics/Athena/Effects/Particles/Skydiving/NS_Contrail_ParallelComic/NS_Contrail_ParallelComic
dependencies {
 usedDataInterfaces [
  &NiagaraDataInterfaceArrayFloat
  &NiagaraDataInterfaceVectorCurve
  &NiagaraDataInterfaceCurve
  &NiagaraDataInterfaceParticleRead
  &NiagaraDataInterfaceArrayInt32
  &NiagaraDataInterfaceArrayFloat3
 ]
 usedDynamicInputs [&Waveform, &SetBoolByFloatComparison]
 usedModules [
  &SystemState
  &self:SystemSpawnScript.SystemScriptSource.SystemScriptGraph.NiagaraNodeAssignment_0.SetVariables_#1
  &self:Dive_Control_Attribute
  &Timeline
 ]
}
emitters {
 Cross_Webs {
  emitterUpdateScript [
   &EmitterState(
    Inactive Response=::NewEnumerator1
    Life Cycle Mode=::NewEnumerator1
    Loop Behavior=::NewEnumerator1
    Loop Duration Mode=::NewEnumerator1
    UseLoopDelay=false
   )
   &SpawnBurst_Instantaneous(Spawn Count=60, Use Spawn Probability=false, UseLoopCountLimit=false)
  ]
  particleSpawnScript [
   &FN_SetSpawnGroup(Inclusivity=true, Normalize Spawn Group=true, Particles Per Group=10, Pattern=::NewEnumerator0)
   &InitializeRibbon(UsePositionOffset=false, Write Material Random=true)
   &DynamicMaterialParameters(
    Index 0 Param 1=0.125
    Index 0 Param 2=&SelectFloatFromArray(Direct Array Index=$Particles.SpawnGroup, Float Selection Array=DataInterface() {propertyValues {FloatData [0.25,1,0.5]}})
    Module.Param0WriteEnabled=true
    Module.Param1WriteEnabled=true
    Module.Param2WriteEnabled=false
    Module.Param3WriteEnabled=false
   )
  ]
  particleUpdateScript [
   &ParticleState(Kill Particles When Lifetime Has Elapsed=false, Let Infinitely Lived Particles Die When Emitter Deactivates=true)
   &self:ReadWebPosition(
    New Particle Attribute Reader=DataInterface() {propertyValues {EmitterBinding {bindingMode=Other, emitterName=Webs}}}
    Spawn Group:NiagaraInt32=&ModuloInt(A=$Particles.SpawnGroup, B=6)
    Start Indicies=DataInterface() {propertyValues {GpuSyncMode=SyncCpuToGpu, MaxElements=0, IntData [6,7,8,26,27,28]}}
    Stop Indicies=DataInterface() {propertyValues {GpuSyncMode=SyncCpuToGpu, MaxElements=0, IntData [15,16,17,16,17,18]}}
   ) {module=ReadWebPosition}
   /SetVariables(Particles.Origin:Vector3f=&Lerp_Vector(A=$Particles.StartSampledPosition, B=$Particles.StopSampledPosition, Alpha=$Particles.Normalized Spawn Group)) {isSetParameters=true, module=SetVariables_#2}
   &PendulumSetup(
    Module.Enable Spring Driver=false
    Pendulum Length:NiagaraFloat=&FloatFromCurve(
     FloatCurve=DataInterface() {propertyValues {Curve {keys [
         {interpMode=RCIM_Cubic, leaveTangentWeight=0.5867, tangentMode=RCTM_User, time=0, value=0.025}
         {arriveTangentWeight=0.3375, interpMode=RCIM_Cubic, leaveTangentWeight=0.3375, tangentMode=RCTM_User, time=0.5, value=1}
         {arriveTangentWeight=0.617, interpMode=RCIM_Cubic, tangentMode=RCTM_User, time=1, value=0}
        ]}}}
     CurveIndex=$Particles.Normalized Spawn Group
     Scale Curve=4
    )
    Pendulum Pivot:Vector3f=$Particles.Origin
    Rigid Pendulum=false
   )
   &GravityForce(Gravity=<0 0 980>)
   &SolveForcesAndVelocity(Clamp Velocity=false, Limit Acceleration=false)
   &PendulumConstraint(Calculate and Output Potential Energy=true)
   &CalculateAccurateVelocity()
   /SetVariables(
    Particles.RibbonID=/Niagara/DynamicInputs/NiagaraID/MakeNiagaraIDFromInt(Index:NiagaraInt32=$Particles.SpawnGroup, Acquire Tag:NiagaraInt32=0)
    Particles.RibbonLinkOrder:NiagaraFloat=$Particles.Normalized Spawn Group
   ) {isSetParameters=true, module=SetVariables_#3}
   &ScaleRibbonWidth(Ribbon Width Scale=&FloatFromCurve(
     FloatCurve=DataInterface() {propertyValues {Curve {keys [
         {
          arriveTangent=2.7436
          arriveTangentWeight=0.0667
          interpMode=RCIM_Cubic
          leaveTangent=2.7436
          leaveTangentWeight=0.3432
          tangentMode=RCTM_User
          time=0
          value=0.5
         }
         {arriveTangentWeight=0.1667, interpMode=RCIM_Cubic, leaveTangentWeight=0.1667, tangentMode=RCTM_User, time=0.5, value=1}
         {arriveTangent=-3.4038, arriveTangentWeight=0.3432, interpMode=RCIM_Cubic, leaveTangent=-3.4038, tangentMode=RCTM_User, time=1, value=0.5}
        ]}}}
     CurveIndex=$Particles.Normalized Spawn Group
    ))
   &ScaleColor(
    Scale Alpha=&FloatFromCurve(FloatCurve=DataInterface() {propertyValues {Curve {keys [<at 0 0>,<at 0.3 1>]}}}, CurveIndex=$System.Timeline.Time)
    ScaleA=true
    ScaleRGB=true
   )
  ]
  rendererClasses [&NiagaraRibbonRendererProperties]
  renderers [{rendererClass=&NiagaraRibbonRendererProperties}]
 }
 Cross_Webs001 {
  emitterUpdateScript [
   &EmitterState(
    Inactive Response=::NewEnumerator1
    Life Cycle Mode=::NewEnumerator1
    Loop Behavior=::NewEnumerator1
    Loop Duration Mode=::NewEnumerator1
    UseLoopDelay=false
   )
   &SpawnBurst_Instantaneous(Spawn Count=60, Use Spawn Probability=false, UseLoopCountLimit=false)
  ]
  particleSpawnScript [
   &FN_SetSpawnGroup(Inclusivity=true, Normalize Spawn Group=true, Particles Per Group=10, Pattern=::NewEnumerator0)
   &InitializeRibbon(UsePositionOffset=false, Write Material Random=true)
   &DynamicMaterialParameters(
    Index 0 Param 1=0.125
    Index 0 Param 2=&SelectFloatFromArray(Direct Array Index=$Particles.SpawnGroup, Float Selection Array=DataInterface() {propertyValues {FloatData [0.25,1,0.5]}})
    Module.Param0WriteEnabled=true
    Module.Param1WriteEnabled=true
    Module.Param2WriteEnabled=false
    Module.Param3WriteEnabled=false
   )
  ]
  particleUpdateScript [
   &ParticleState(Kill Particles When Lifetime Has Elapsed=false, Let Infinitely Lived Particles Die When Emitter Deactivates=true)
   &self:ReadWebPosition(
    New Particle Attribute Reader=DataInterface() {propertyValues {EmitterBinding {bindingMode=Other, emitterName=Webs001}}}
    Spawn Group:NiagaraInt32=&ModuloInt(A=$Particles.SpawnGroup, B=6)
    Start Indicies=DataInterface() {propertyValues {GpuSyncMode=SyncCpuToGpu, MaxElements=0, IntData [6,7,8,26,27,28]}}
    Stop Indicies=DataInterface() {propertyValues {GpuSyncMode=SyncCpuToGpu, MaxElements=0, IntData [15,16,17,16,17,18]}}
   ) {module=ReadWebPosition}
   /SetVariables(Particles.Origin:Vector3f=&Lerp_Vector(A=$Particles.StartSampledPosition, B=$Particles.StopSampledPosition, Alpha=$Particles.Normalized Spawn Group)) {isSetParameters=true, module=SetVariables_#2}
   &PendulumSetup(
    Module.Enable Spring Driver=false
    Pendulum Length:NiagaraFloat=&FloatFromCurve(
     FloatCurve=DataInterface() {propertyValues {Curve {keys [
         {interpMode=RCIM_Cubic, leaveTangentWeight=0.5867, tangentMode=RCTM_User, time=0, value=0.025}
         {arriveTangentWeight=0.3375, interpMode=RCIM_Cubic, leaveTangentWeight=0.3375, tangentMode=RCTM_User, time=0.5, value=1}
         {arriveTangentWeight=0.617, interpMode=RCIM_Cubic, tangentMode=RCTM_User, time=1, value=0}
        ]}}}
     CurveIndex=$Particles.Normalized Spawn Group
     Scale Curve=4
    )
    Pendulum Pivot:Vector3f=$Particles.Origin
    Rigid Pendulum=false
   )
   &GravityForce(Gravity=<0 0 980>)
   &SolveForcesAndVelocity(Clamp Velocity=false, Limit Acceleration=false)
   &PendulumConstraint(Calculate and Output Potential Energy=true)
   &CalculateAccurateVelocity()
   /SetVariables(
    Particles.RibbonID=&MakeNiagaraIDFromInt(Index:NiagaraInt32=$Particles.SpawnGroup, Acquire Tag:NiagaraInt32=0)
    Particles.RibbonLinkOrder:NiagaraFloat=$Particles.Normalized Spawn Group
   ) {isSetParameters=true, module=SetVariables_#3}
   &ScaleRibbonWidth(Ribbon Width Scale=&FloatFromCurve(
     FloatCurve=DataInterface() {propertyValues {Curve {keys [
         {
          arriveTangent=2.7436
          arriveTangentWeight=0.0667
          interpMode=RCIM_Cubic
          leaveTangent=2.7436
          leaveTangentWeight=0.3432
          tangentMode=RCTM_User
          time=0
          value=0.5
         }
         {arriveTangentWeight=0.1667, interpMode=RCIM_Cubic, leaveTangentWeight=0.1667, tangentMode=RCTM_User, time=0.5, value=1}
         {arriveTangent=-3.4038, arriveTangentWeight=0.3432, interpMode=RCIM_Cubic, leaveTangent=-3.4038, tangentMode=RCTM_User, time=1, value=0.5}
        ]}}}
     CurveIndex=$Particles.Normalized Spawn Group
    ))
   &ScaleColor(
    Scale Alpha=&FloatFromCurve(FloatCurve=DataInterface() {propertyValues {Curve {keys [<at 0 0>,<at 0.3 1>]}}}, CurveIndex=$System.Timeline.Time)
    ScaleA=true
    ScaleRGB=true
   )
  ]
  rendererClasses [&NiagaraRibbonRendererProperties]
  renderers [{rendererClass=&NiagaraRibbonRendererProperties}]
 }
 HandWraps {
  emitterUpdateScript [
   &EmitterState(
    Inactive Response=::NewEnumerator1
    Life Cycle Mode=::NewEnumerator1
    Loop Behavior=::NewEnumerator1
    Loop Delay=0.5502
    Loop Duration Mode=::NewEnumerator1
    UseLoopDelay=true
   )
   &SpawnBurst_Instantaneous(Spawn Count=8, Use Spawn Probability=false, UseLoopCountLimit=false)
   &SpawnBurst_Instantaneous(Spawn Count=8, Spawn Group=1, Use Spawn Probability=false, UseLoopCountLimit=false) {module=SpawnBurst_Instantaneous001}
  ]
  particleSpawnScript [
   &InitializeParticle(
    Color Mode=::NewEnumerator0
    Mesh Scale Mode=::NewEnumerator1
    Mesh Uniform Scale Max=0.08
    Mesh Uniform Scale Min=0.06
    UsePositionOffset=false
    Write Material Random=true
   )
   &DynamicMaterialParameters(
    Index 0 Param 1=0.5
    Index 0 Param 2=&RandomRangeFloat(Minimum=0, Maximum=1)
    Index 1 Param 3=4
    Index 1 Param 4=4
    Module.Param0WriteEnabled=true
    Module.Param0WriteEnabled1=false
    Module.Param1WriteEnabled=true
    Module.Param1WriteEnabled1=false
    Module.Param2WriteEnabled=false
    Module.Param2WriteEnabled1=true
    Module.Param3WriteEnabled=false
    Module.Param3WriteEnabled1=true
    Write Parameter Index 1=true
   )
   &SkeletalMeshLocation(
    Bone / Socket Index=&SelectIntFromArray(Direct Array Index=&SpawnGroup(), Int Selection Array=DataInterface() {propertyValues {IntData [0,3]}})
    Bone Sampling Mode=::NewEnumerator4
    MaskSpawnGroup=false
    Position Sampling=::NewEnumerator2
    Skeletal Mesh=$User.Player
   ) {module=SkeletalMeshLocation001}
   &InitialMeshOrientation(
    Mesh Axis To Orient=<0 0 1>
    Mesh Orientation Mode=::NewEnumerator2
    Orientation Vector=&RotateVectorByQuaternion(Vector To Rotate=<1 0 0>, Quaternion=$Particles.SkeletalMeshLocation001.SampledRotation)
    Rotation=&RandomRangeVector(Minimum=<0 0 0>, Maximum=<1 1 0>)
    Use Rotation Vector=true
   )
  ]
  particleUpdateScript [
   &ParticleState(Kill Particles When Lifetime Has Elapsed=false, Let Infinitely Lived Particles Die When Emitter Deactivates=true)
   &SkeletalMeshLocation(
    Bone / Socket Index=&SelectIntFromArray(Direct Array Index=&SpawnGroup(), Int Selection Array=DataInterface() {propertyValues {IntData [0,3]}})
    Bone Sampling Mode=::NewEnumerator4
    MaskSpawnGroup=false
    Skeletal Mesh=$User.Player
   )
  ]
  rendererClasses [&NiagaraMeshRendererProperties]
  renderers [{rendererClass=&NiagaraMeshRendererProperties, rendererIndex=0}]
 }
 Smoke {
  emitterUpdateScript [
   &EmitterState(Life Cycle Mode=::NewEnumerator1, Loop Behavior=::NewEnumerator1, Loop Duration=2.0484, UseLoopDelay=false)
   &SpawnBurst_Instantaneous(Spawn Count=7, Use Spawn Probability=false) {module=SpawnBurst_Instantaneous001}
  ]
  particleSpawnScript [
   /Niagara/Modules/Spawn/Initialization/InitializeParticle(
    Color=<rgba 0.5255 0.5255 0.5255 1>
    Lifetime=&UniformRangedFloat(Minimum=0.2, Maximum=0.6)
    Module.Write Color=true
    Module.Write Lifetime=true
    Module.Write Mass=true
    Module.Write Material Random=true
    Module.Write Position=false
    Module.Write Scale=false
    Module.Write SpriteRotation=true
    Module.Write SpriteSize=true
    Position=&SimulationPosition()
    Sprite Rotation=&UniformRangedFloat(Minimum=0, Maximum=360)
    Sprite Size=&Vector2DFromFloat(Value=&UniformRangedFloat(Minimum=19.8, Maximum=24.3))
   )
   /SetVariables(Particles.InitSpriteRotRate=&UniformRangedFloat(Minimum=-1, Maximum=1)) {isSetParameters=true}
   &AccelerationForce(Acceleration=&UniformRangedVector(Minimum=<0 0 -25>, Maximum=<0 0 -75>))
   &SphereLocation(
    MaskSpawnGroup=false
    Offset=<0 0 500>
    Override Seed=false
    Sphere Radius=3
    Surface Only=true
    UseNonUniformScale=false
    UseOffset=true
    UseSphereOrientation=false
   )
   &AddVelocityFromPoint(Module.Use Falloff Distance=false, Velocity Strength=&RandomRangeFloat(Minimum=100, Maximum=150))
   &ApplyInitialForces()
   &CameraOffset(Camera Offset Amount=&RandomRangeFloat(Minimum=0, Maximum=15))
   &AddVelocity(Velocity=<-100 0 -100>) {module=AddVelocity001}
  ]
  particleUpdateScript [
   &ParticleState()
   &SpriteRotationRate(Rotation Rate=&Multiply_Float(
     A=$Particles.InitSpriteRotRate
     B=&FloatFromCurve(
      FloatCurve=DataInterface() {propertyValues {Curve {keys [
          {arriveTangent=-1.9214, interpMode=RCIM_Cubic, leaveTangent=-1.9214, tangentMode=RCTM_User, time=-0.0046, value=1.0113}
          {arriveTangent=-0.2496, interpMode=RCIM_Cubic, leaveTangent=-0.2496, tangentMode=RCTM_User, time=0.9896, value=0.16}
         ]}}}
      CurveIndex=$Emitter.Age
     )
    ))
   &ScaleColor(
    Scale Alpha=&FloatFromCurve(FloatCurve=DataInterface() {propertyValues {Curve {keys [<at 0 0>,<at 0.2049 0.9765>,<at 0.7791 0.9107>,<at 1 0>]}}})
    ScaleA=true
    ScaleRGB=true
   ) {module=ScaleColor001}
   &Drag(Drag=5, Use Linear Drag=true, Use Rotational Drag=false)
   &SolveForcesAndVelocity(Clamp Velocity=false, Limit Acceleration=false)
   &ScaleSpriteSize(Scale Factor=&Vector2DFromFloat(Value=&FloatFromCurve(FloatCurve=DataInterface() {propertyValues {Curve {keys [
          {arriveTangent=5.2354, interpMode=RCIM_Cubic, leaveTangent=5.2354, tangentMode=RCTM_User, time=-0.0046, value=0.25}
          <at 0.1439 1.4923>
          <at 1.0011 3.4809 RCIM_Cubic>
         ]}}})))
   &DynamicMaterialParameters(
    Index 0 Param 1=&RandomRangeFloat(Minimum=0, Maximum=1)
    Index 0 Param 2=&FloatFromCurve(FloatCurve=DataInterface() {propertyValues {Curve {keys [
         {arriveTangent=0.0453, interpMode=RCIM_Cubic, tangentMode=RCTM_Break, time=0, value=-1}
         {arriveTangent=0.0604, interpMode=RCIM_Cubic, leaveTangent=0.0604, tangentMode=RCTM_User, time=1, value=-1}
        ]}}})
    Index 0 Param 3=&FloatFromCurve(FloatCurve=DataInterface() {propertyValues {Curve {keys [{interpMode=RCIM_Cubic, leaveTangent=4.6779, time=0.2851, value=0}]}}})
    Index 1 Param 1=&FloatFromCurve(
     FloatCurve=DataInterface() {propertyValues {Curve {keys [<at 0 0>,<at 0.25 1>]}}}
     Scale Curve=&RandomRangeFloat(Minimum=2, Maximum=1)
    )
    Index 1 Param 2=&FloatFromCurve(
     FloatCurve=DataInterface() {propertyValues {Curve {keys [
         {interpMode=RCIM_Cubic, leaveTangent=3.7422, tangentMode=RCTM_Break, time=0, value=0.2}
         {interpMode=RCIM_Cubic, leaveTangent=0.9066, time=1, value=2}
        ]}}}
     Scale Curve=&RandomRangeFloat(Minimum=0.5, Maximum=1)
    )
    Index 1 Param 3=&FloatFromCurve(FloatCurve=DataInterface() {propertyValues {Curve {keys [
         {interpMode=RCIM_Cubic, leaveTangent=2.8445, tangentMode=RCTM_Break, time=0, value=0.2}
         {interpMode=RCIM_Cubic, leaveTangent=63.2978, time=0.25, value=1}
        ]}}})
    Index 1 Param 4=&FloatFromCurve(
     FloatCurve=DataInterface() {propertyValues {Curve {keys [<at 0 1>,<at 1 0.8>]}}}
     Scale Curve=&RandomRangeFloat(Minimum=0.5, Maximum=1)
    )
    Index 2 Param 1=&FloatFromCurve(FloatCurve=DataInterface() {propertyValues {Curve {keys [{interpMode=RCIM_Cubic, leaveTangent=3.2309, tangentMode=RCTM_Break, time=0.1, value=0}, <at 1 1>]}}})
    Index 2 Param 2=&FloatFromCurve(
     FloatCurve=DataInterface() {propertyValues {Curve {keys [<at 0 0.2 RCIM_Cubic>,<at 1 0.8 RCIM_Cubic>]}}}
     Scale Curve=&RandomRangeFloat(Minimum=0.2, Maximum=1.2)
    )
    Index 2 Param 3=&FloatFromCurve(FloatCurve=DataInterface() {propertyValues {Curve {keys [
         {interpMode=RCIM_Cubic, tangentMode=RCTM_Break, time=0, value=1}
         {interpMode=RCIM_Cubic, leaveTangent=11.1233, tangentMode=RCTM_Break, time=1, value=2}
        ]}}})
    Index 2 Param 4=&FloatFromCurve(
     FloatCurve=DataInterface() {propertyValues {Curve {keys [<at 0 1 RCIM_Cubic>, {interpMode=RCIM_Cubic, leaveTangent=-1.2956, tangentMode=RCTM_Break, time=1, value=0.75}]}}}
     Scale Curve=0.8
    )
    Module.Param0WriteEnabled=true
    Module.Param0WriteEnabled1=true
    Module.Param0WriteEnabled2=true
    Module.Param1WriteEnabled=false
    Module.Param1WriteEnabled1=true
    Module.Param1WriteEnabled2=true
    Module.Param2WriteEnabled=true
    Module.Param2WriteEnabled1=true
    Module.Param2WriteEnabled2=true
    Module.Param3WriteEnabled=true
    Module.Param3WriteEnabled1=true
    Module.Param3WriteEnabled2=true
    Write Parameter Index 1=true
    Write Parameter Index 2=true
   )
  ]
  rendererClasses [&NiagaraSpriteRendererProperties]
  renderers [{rendererClass=&NiagaraSpriteRendererProperties}]
 }
 Thwip_Impact {
  emitterUpdateScript [
   &EmitterState(Life Cycle Mode=::NewEnumerator1, Loop Behavior=::NewEnumerator1, Loop Delay=0.5502, Loop Duration=0.1, UseLoopDelay=true)
   &SpawnBurst_Instantaneous(Spawn Count=3, Use Spawn Probability=false, UseLoopCountLimit=false)
   &SpawnBurst_Instantaneous(Spawn Count=3, Spawn Group=1, Use Spawn Probability=false, UseLoopCountLimit=false) {module=SpawnBurst_Instantaneous001}
   &SampleSkeletalMesh(
    AdditionalYawPitchRoll=false
    Bone Sampling Mode:ENiagaraBoneSamplingMode=::NewEnumerator4
    Sampled Position Offset:Vector3f=<0 0 20>
    Skeletal Mesh:NiagaraDataInterfaceSkeletalMesh=$User.Player
   )
   &SampleSkeletalMesh(
    AdditionalYawPitchRoll=false
    Bone / Socket Index=3
    Bone Sampling Mode:ENiagaraBoneSamplingMode=::NewEnumerator4
    Sampled Position Offset:Vector3f=<0 0 20>
    Skeletal Mesh:NiagaraDataInterfaceSkeletalMesh=$User.Player
   ) {module=SampleSkeletalMesh001}
  ]
  particleSpawnScript [
   /Niagara/Modules/Spawn/Initialization/V2/InitializeParticle(
    Color Mode=::NewEnumerator0
    Lifetime Max=0.55
    Lifetime Min=0.3
    Lifetime Mode=::NewEnumerator1
    Sprite Rotation Mode=::NewEnumerator2
    Sprite Size Mode=::NewEnumerator3
    Uniform Sprite Size=50
    UsePositionOffset=false
    Write Material Random=true
   )
   &SubUVAnimation(End Frame=3, SubUV Animation Mode:ENiagara_SubUVLookupMode=::NewEnumerator2)
  ]
  particleUpdateScript [
   &ParticleState()
   &CameraOffset(Camera Offset Amount=&FloatFromCurve(FloatCurve=DataInterface() {propertyValues {Curve {keys [<at 0 25>,<at 1 75>]}}}))
   /SetVariables(Particles.Position=&MaskVectorBySpawnGroup(Masked Vector=$Emitter.SampleSkeletalMesh.SampledPosition, Passthrough Vector=$Emitter.SampleSkeletalMesh001.SampledPosition)) {isSetParameters=true}
   &InheritSourceMovement(
    Applied Position Scale=&Multiply_VectorByFloat(
     Vector=<1 1 1>
     Float=&FloatFromCurve(FloatCurve=DataInterface() {propertyValues {Curve {keys [
          {arriveTangent=-8.378, interpMode=RCIM_Cubic, leaveTangent=-8.378, leaveTangentWeight=0.4714, tangentMode=RCTM_User, time=0.75, value=1}
          {arriveTangentWeight=0.2041, interpMode=RCIM_Cubic, tangentMode=RCTM_User, time=1, value=0}
         ]}}})
    )
    ApplyForce=false
    ApplyPosition=true
    ApplyPositionFalloff=false
   )
   &SolveForcesAndVelocity(Clamp Velocity=false, Limit Acceleration=false)
   &DynamicMaterialParameters(
    Index 0 Param 1=&FloatFromCurve(FloatCurve=DataInterface() {propertyValues {Curve {keys [{interpMode=RCIM_Cubic, leaveTangent=-3.0153, tangentMode=RCTM_Break, time=0.2, value=1}, <at 1 0>]}}})
    Module.Param0WriteEnabled=true
    Module.Param1WriteEnabled=true
    Module.Param2WriteEnabled=true
    Module.Param3WriteEnabled=true
   )
   &ScaleSpriteSize(Uniform Curve Sprite Scale=DataInterface() {propertyValues {Curve {keys [
        {arriveTangent=3.1253, interpMode=RCIM_Cubic, leaveTangent=3.1253, leaveTangentWeight=0.4714, tangentMode=RCTM_User, time=0, value=0.5}
        {arriveTangentWeight=0.4714, interpMode=RCIM_Cubic, tangentMode=RCTM_User, time=1, value=1.5}
       ]}}})
  ]
  rendererClasses [&NiagaraSpriteRendererProperties]
  renderers [{rendererClass=&NiagaraSpriteRendererProperties}]
 }
 Thwip_Proj {
  emitterUpdateScript [
   &EmitterState(Life Cycle Mode=::NewEnumerator1, Loop Behavior=::NewEnumerator1, Loop Duration=1, UseLoopDelay=false)
   &SpawnBurst_Instantaneous(Use Spawn Probability=false, UseLoopCountLimit=false)
   &SpawnBurst_Instantaneous(Spawn Group=1, Use Spawn Probability=false, UseLoopCountLimit=false) {module=SpawnBurst_Instantaneous001}
   &SampleSkeletalMesh(
    AdditionalYawPitchRoll=false
    Bone Sampling Mode:ENiagaraBoneSamplingMode=::NewEnumerator4
    Skeletal Mesh:NiagaraDataInterfaceSkeletalMesh=$User.Player
   )
   &SampleSkeletalMesh(
    AdditionalYawPitchRoll=false
    Bone / Socket Index=3
    Bone Sampling Mode:ENiagaraBoneSamplingMode=::NewEnumerator4
    Skeletal Mesh:NiagaraDataInterfaceSkeletalMesh=$User.Player
   ) {module=SampleSkeletalMesh001}
  ]
  particleSpawnScript [&InitializeParticle(
    Color Mode=::NewEnumerator0
    Mesh Scale=<0.25 0.1 0.1>
    Mesh Scale Mode=::NewEnumerator4
    Position Offset=<0 0 500>
    UsePositionOffset=true
    Write Material Random=true
   )]
  particleUpdateScript [
   &ParticleState()
   &PointAttractionForce(
    Attraction Radius=750
    Attraction Strength=150
    Attractor Position:NiagaraPosition=&MaskVectorBySpawnGroup(Masked Vector=$Emitter.SampleSkeletalMesh.SampledPosition, Passthrough Vector=$Emitter.SampleSkeletalMesh001.SampledPosition)
    Kill Within Radius=true
    Module.Use Falloff=true
    UseAttractorOffset=false
   )
   &SolveForcesAndVelocity(Clamp Velocity=false, Limit Acceleration=false)
  ]
  rendererClasses [&NiagaraMeshRendererProperties]
  renderers [{rendererClass=&NiagaraMeshRendererProperties, rendererIndex=0}]
 }
 Webs {
  emitterUpdateScript [
   &EmitterState(
    Inactive Response=::NewEnumerator1
    Life Cycle Mode=::NewEnumerator1
    Loop Behavior=::NewEnumerator1
    Loop Duration Mode=::NewEnumerator1
    UseLoopDelay=false
   )
   &SpawnBurst_Instantaneous(Spawn Count=30, Use Spawn Probability=false, UseLoopCountLimit=false)
   &SampleSkeletalMesh(
    AdditionalYawPitchRoll=false
    Bone Sampling Mode:ENiagaraBoneSamplingMode=::NewEnumerator4
    Skeletal Mesh:NiagaraDataInterfaceSkeletalMesh=$User.Player
   )
  ]
  particleSpawnScript [
   &FN_SetSpawnGroup(Inclusivity=true, Normalize Spawn Group=true, Particles Per Group=10, Pattern=::NewEnumerator0)
   &InitializeRibbon(Ribbon Width=7, UsePositionOffset=false, Write Material Random=true)
   &DynamicMaterialParameters(
    Index 0 Param 1=&SelectFloatFromArray(
     Direct Array Index=&ModuloInt(A=$Particles.SpawnGroup, B=3)
     Float Selection Array=DataInterface() {propertyValues {FloatData [0.25,0.333,0.25]}}
    )
    Index 0 Param 2=&SelectFloatFromArray(
     Direct Array Index=&ModuloInt(A=$Particles.SpawnGroup, B=3)
     Float Selection Array=DataInterface() {propertyValues {FloatData [0.25,1,0.333]}}
    )
    Module.Param0WriteEnabled=true
    Module.Param1WriteEnabled=true
    Module.Param2WriteEnabled=false
    Module.Param3WriteEnabled=false
   )
  ]
  particleUpdateScript [
   &ParticleState(Kill Particles When Lifetime Has Elapsed=false, Let Infinitely Lived Particles Die When Emitter Deactivates=true)
   /SetVariables(Particles.Origin:Vector3f=&Add_Vector(
     A=&Lerp_Vector(
      A=$Emitter.SampleSkeletalMesh.SampledPosition
      B=&Add_Vector(
       A=$Emitter.SampleSkeletalMesh.SampledPosition
       B=&Multiply_Vector(
        A=&VectorFromCurve(
         VectorCurve=DataInterface() {propertyValues {
           XCurve {keys [
             {interpMode=RCIM_Cubic, leaveTangentWeight=0.3806, tangentMode=RCTM_User, time=0, value=0}
             {arriveTangent=0.9139, arriveTangentWeight=0.3333, interpMode=RCIM_Cubic, leaveTangent=0.9139, tangentMode=RCTM_User, time=1, value=1}
            ]}
           YCurve {keys [
             {interpMode=RCIM_Cubic, leaveTangentWeight=0.3439, tangentMode=RCTM_User, time=0, value=0}
             {arriveTangent=2.8572, arriveTangentWeight=0.3333, interpMode=RCIM_Cubic, leaveTangent=2.8572, tangentMode=RCTM_User, time=1, value=1}
            ]}
           ZCurve {keys [<at 0 1>,<at 1 1>]}
          }}
         CurveIndex=$Particles.Normalized Spawn Group
        )
        B=&Add_Vector(
         A=&MakeVector(
          X=&SelectFloatFromArray(
           Direct Array Index=&ModuloInt(A=$Particles.SpawnGroup, B=3)
           Float Selection Array=DataInterface() {propertyValues {FloatData [-25,0,20]}}
          )
          Y=&SelectFloatFromArray(
           Direct Array Index=&ModuloInt(A=$Particles.SpawnGroup, B=3)
           Float Selection Array=DataInterface() {propertyValues {FloatData [30,0,-20]}}
          )
          Z=&SelectFloatFromArray(
           Direct Array Index=&ModuloInt(A=$Particles.SpawnGroup, B=3)
           Float Selection Array=DataInterface() {propertyValues {FloatData [160,175,150]}}
          )
         )
         B=<0 -20 0>
        )
       )
      )
      Alpha=&Multiply_Float(A=$Particles.Normalized Spawn Group, B=$System.Timeline.Time)
     )
     B=&Multiply_Vector(
      A=<-35 -20 0>
      B=&VectorFromFloat(Value=&FloatFromCurve(
        FloatCurve=DataInterface() {propertyValues {Curve {keys [{tangentMode=RCTM_User, time=0, value=0}, {tangentMode=RCTM_User, time=1, value=1}]}}}
        CurveIndex=$Particles.Normalized Spawn Group
        Scale Curve=$System.Dive
       ))
     )
    )) {isSetParameters=true, module=SetVariables_#4}
   &ScaleRibbonWidth(Ribbon Width Scale=&FloatFromCurve(
     FloatCurve=DataInterface() {propertyValues {Curve {keys [
         {arriveTangentWeight=0.0667, interpMode=RCIM_Cubic, leaveTangentWeight=0.3335, tangentMode=RCTM_User, time=0.2, value=1}
         {arriveTangent=-3.4913, arriveTangentWeight=0.4269, interpMode=RCIM_Cubic, leaveTangent=-3.4913, tangentMode=RCTM_User, time=1, value=0}
        ]}}}
     CurveIndex=$Particles.Normalized Spawn Group
    ))
   /SetVariables(
    Particles.RibbonID=&MakeNiagaraIDFromInt(Index:NiagaraInt32=$Particles.SpawnGroup, Acquire Tag:NiagaraInt32=0)
    Particles.RibbonLinkOrder:NiagaraFloat=$Particles.Normalized Spawn Group
   ) {isSetParameters=true, module=SetVariables_#3}
   &PendulumSetup(
    Module.Enable Spring Driver=false
    Pendulum Length:NiagaraFloat=&FloatFromCurve(
     FloatCurve=DataInterface() {propertyValues {Curve {keys [
         {arriveTangent=5.0354, interpMode=RCIM_Cubic, leaveTangent=5.0354, leaveTangentWeight=0.5867, tangentMode=RCTM_User, time=0, value=0.1}
         {arriveTangentWeight=0.3375, interpMode=RCIM_Cubic, leaveTangentWeight=0.3375, tangentMode=RCTM_User, time=0.5, value=1}
        ]}}}
     CurveIndex=$Particles.Normalized Spawn Group
     Scale Curve=5
    )
    Pendulum Pivot:Vector3f=$Particles.Origin
    Rigid Pendulum=false
   )
   &VectorNoiseForce(Force Amount=&Lerp_Float(A=1500, B=3000, Alpha=$System.Dive), Module.Use Random Noise Time Interval=false)
   &GravityForce(Gravity=<0 0 980>)
   &CurlNoiseForce(
    Noise Frequency=100
    Noise Strength=&Lerp_Float(A=5500, B=11000, Alpha=$System.Dive)
    Pan Noise=true
    Pan Noise Field=<0 0 1500>
   )
   &SolveForcesAndVelocity(Clamp Velocity=false, Limit Acceleration=false)
   &PendulumConstraint(Calculate and Output Potential Energy=true)
   &CalculateAccurateVelocity()
   &ScaleColor(
    Scale Alpha=&FloatFromCurve(FloatCurve=DataInterface() {propertyValues {Curve {keys [<at 0 0>,<at 0.3 1>]}}}, CurveIndex=$System.Timeline.Time)
    ScaleA=true
    ScaleRGB=true
   )
  ]
  rendererClasses [&NiagaraRibbonRendererProperties]
  renderers [{rendererClass=&NiagaraRibbonRendererProperties}]
 }
 Webs001 {
  emitterUpdateScript [
   &EmitterState(
    Inactive Response=::NewEnumerator1
    Life Cycle Mode=::NewEnumerator1
    Loop Behavior=::NewEnumerator1
    Loop Duration Mode=::NewEnumerator1
    UseLoopDelay=false
   )
   &SpawnBurst_Instantaneous(Spawn Count=30, Use Spawn Probability=false, UseLoopCountLimit=false)
   &SampleSkeletalMesh(
    AdditionalYawPitchRoll=false
    Bone / Socket Index=3
    Bone Sampling Mode:ENiagaraBoneSamplingMode=::NewEnumerator4
    Skeletal Mesh:NiagaraDataInterfaceSkeletalMesh=$User.Player
   )
  ]
  particleSpawnScript [
   &FN_SetSpawnGroup(Inclusivity=true, Normalize Spawn Group=true, Particles Per Group=10, Pattern=::NewEnumerator0)
   &InitializeRibbon(Ribbon Width=7, UsePositionOffset=false, Write Material Random=true)
   &DynamicMaterialParameters(
    Index 0 Param 1=&SelectFloatFromArray(
     Direct Array Index=&ModuloInt(A=$Particles.SpawnGroup, B=3)
     Float Selection Array=DataInterface() {propertyValues {FloatData [0.25,0.333,0.25]}}
    )
    Index 0 Param 2=&SelectFloatFromArray(
     Direct Array Index=&ModuloInt(A=$Particles.SpawnGroup, B=3)
     Float Selection Array=DataInterface() {propertyValues {FloatData [0.25,1,0.333]}}
    )
    Module.Param0WriteEnabled=true
    Module.Param1WriteEnabled=true
    Module.Param2WriteEnabled=false
    Module.Param3WriteEnabled=false
   )
  ]
  particleUpdateScript [
   &ParticleState(Kill Particles When Lifetime Has Elapsed=false, Let Infinitely Lived Particles Die When Emitter Deactivates=true)
   /SetVariables(Particles.Origin:Vector3f=&Add_Vector(
     A=&Lerp_Vector(
      A=$Emitter.SampleSkeletalMesh.SampledPosition
      B=&Add_Vector(
       A=$Emitter.SampleSkeletalMesh.SampledPosition
       B=&Multiply_Vector(
        A=&VectorFromCurve(
         VectorCurve=DataInterface() {propertyValues {
           XCurve {keys [
             {interpMode=RCIM_Cubic, leaveTangentWeight=0.3806, tangentMode=RCTM_User, time=0, value=0}
             {arriveTangent=0.9139, arriveTangentWeight=0.3333, interpMode=RCIM_Cubic, leaveTangent=0.9139, tangentMode=RCTM_User, time=1, value=1}
            ]}
           YCurve {keys [
             {interpMode=RCIM_Cubic, leaveTangentWeight=0.3439, tangentMode=RCTM_User, time=0, value=0}
             {arriveTangent=2.8572, arriveTangentWeight=0.3333, interpMode=RCIM_Cubic, leaveTangent=2.8572, tangentMode=RCTM_User, time=1, value=1}
            ]}
           ZCurve {keys [<at 0 1>,<at 1 1>]}
          }}
         CurveIndex=$Particles.Normalized Spawn Group
        )
        B=&Add_Vector(
         A=&MakeVector(
          X=&SelectFloatFromArray(
           Direct Array Index=&ModuloInt(A=$Particles.SpawnGroup, B=3)
           Float Selection Array=DataInterface() {propertyValues {FloatData [20,0,-25]}}
          )
          Y=&SelectFloatFromArray(
           Direct Array Index=&ModuloInt(A=$Particles.SpawnGroup, B=3)
           Float Selection Array=DataInterface() {propertyValues {FloatData [30,0,-20]}}
          )
          Z=&SelectFloatFromArray(
           Direct Array Index=&ModuloInt(A=$Particles.SpawnGroup, B=3)
           Float Selection Array=DataInterface() {propertyValues {FloatData [160,175,150]}}
          )
         )
         B=<0 -20 0>
        )
       )
      )
      Alpha=&Multiply_Float(A=$Particles.Normalized Spawn Group, B=$System.Timeline.Time)
     )
     B=&Multiply_Vector(
      A=<35 20 0>
      B=&VectorFromFloat(Value=&FloatFromCurve(
        FloatCurve=DataInterface() {propertyValues {Curve {keys [{tangentMode=RCTM_User, time=0, value=0}, {tangentMode=RCTM_User, time=1, value=1}]}}}
        CurveIndex=$Particles.Normalized Spawn Group
        Scale Curve=$System.Dive
       ))
     )
    )) {isSetParameters=true, module=SetVariables_#4}
   &ScaleRibbonWidth(Ribbon Width Scale=&FloatFromCurve(
     FloatCurve=DataInterface() {propertyValues {Curve {keys [
         {arriveTangentWeight=0.0667, interpMode=RCIM_Cubic, leaveTangentWeight=0.3335, tangentMode=RCTM_User, time=0.2, value=1}
         {arriveTangent=-3.4913, arriveTangentWeight=0.4269, interpMode=RCIM_Cubic, leaveTangent=-3.4913, tangentMode=RCTM_User, time=1, value=0}
        ]}}}
     CurveIndex=$Particles.Normalized Spawn Group
    ))
   /SetVariables(
    Particles.RibbonID=&MakeNiagaraIDFromInt(Index:NiagaraInt32=$Particles.SpawnGroup, Acquire Tag:NiagaraInt32=0)
    Particles.RibbonLinkOrder:NiagaraFloat=$Particles.Normalized Spawn Group
   ) {isSetParameters=true, module=SetVariables_#3}
   &PendulumSetup(
    Module.Enable Spring Driver=false
    Pendulum Length:NiagaraFloat=&FloatFromCurve(
     FloatCurve=DataInterface() {propertyValues {Curve {keys [
         {arriveTangent=5.0354, interpMode=RCIM_Cubic, leaveTangent=5.0354, leaveTangentWeight=0.5867, tangentMode=RCTM_User, time=0, value=0.1}
         {arriveTangentWeight=0.3375, interpMode=RCIM_Cubic, leaveTangentWeight=0.3375, tangentMode=RCTM_User, time=0.5, value=1}
        ]}}}
     CurveIndex=$Particles.Normalized Spawn Group
     Scale Curve=5
    )
    Pendulum Pivot:Vector3f=$Particles.Origin
    Rigid Pendulum=false
   )
   &VectorNoiseForce(Force Amount=&Lerp_Float(A=1500, B=3000, Alpha=$System.Dive), Module.Use Random Noise Time Interval=false)
   &GravityForce(Gravity=<0 0 980>)
   &CurlNoiseForce(
    Noise Frequency=100
    Noise Strength=&Lerp_Float(A=5500, B=11000, Alpha=$System.Dive)
    Pan Noise=true
    Pan Noise Field=<0 0 1500>
   )
   &SolveForcesAndVelocity(Clamp Velocity=false, Limit Acceleration=false)
   &PendulumConstraint(Calculate and Output Potential Energy=true)
   &CalculateAccurateVelocity()
   &ScaleColor(
    Scale Alpha=&FloatFromCurve(FloatCurve=DataInterface() {propertyValues {Curve {keys [<at 0 0>,<at 0.3 1>]}}}, CurveIndex=$System.Timeline.Time)
    ScaleA=true
    ScaleRGB=true
   )
  ]
  rendererClasses [&NiagaraRibbonRendererProperties]
  renderers [{rendererClass=&NiagaraRibbonRendererProperties}]
 }
}
summary {
 systemName=NS_Contrail_ParallelComic
 emitters [
  {emitterName=Webs, rendererClasses [&NiagaraRibbonRendererProperties]}
  {emitterName=Cross_Webs, rendererClasses [&NiagaraRibbonRendererProperties]}
  {emitterName=HandWraps, rendererClasses [&NiagaraMeshRendererProperties]}
  {emitterName=Thwip_Proj, rendererClasses [&NiagaraMeshRendererProperties]}
  {emitterName=Thwip_Impact, rendererClasses [&NiagaraSpriteRendererProperties]}
  {emitterName=Webs001, rendererClasses [&NiagaraRibbonRendererProperties]}
  {emitterName=Cross_Webs001, rendererClasses [&NiagaraRibbonRendererProperties]}
  {emitterName=Smoke, rendererClasses [&NiagaraSpriteRendererProperties]}
 ]
 userVariables [
  User.bIsFrontEnd {type=NiagaraBool}
  User.bIsFrontEndPreview {type=NiagaraBool, defaultValue=true}
  User.FrontEndPreview {type=NiagaraFloat, defaultValue=1, description=""}
  User.Player {
   type=NiagaraDataInterfaceSkeletalMesh
   defaultValue {
    struct=NiagaraExt_VariableValue_DataInterface
    value {dataInterface=&self:NiagaraDataInterfaceSkeletalMesh_0}
   }
  }
 ]
}
